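# Sound cue for the fourth Section 03 spire destruction sound in the PistolCanary event, named as a MainHero destroy variant.
# A single wave player plays one destroyed-spire wave, with no randomization, mixing or modulation nodes.
# It uses the shared spire explosion attenuation with a very large max distance, routes to the Fort_Events sound class, and plays at slightly reduced volume.
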SoundCue /PistolCanary_Event_Audio/Assets/Section03/Spires/SC_PistolCanary_Section03_MainHero_Destroy_Spire04
root=WavePlayer_4
nodes [{id=WavePlayer_4, properties {soundWaveAssetPtr=/PistolCanary_Event_Audio/Assets/Section03/wav/PistolCanary_Spire_Sec_03_Destroyed_04}}]
properties {
 attenuationSettings=/PistolCanary_Event_Audio/Assets/Settings/ATT_PistolCanary_Spires_Explosions
 maxDistance=105000
 soundClassObject=&Fort_Events
 volumeMultiplier=0.8
}
